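; BP_HDRChecker is an art-tool actor for checking HDR, with False Color and Out of Range debug views.
; Its construction script switches on ViewMode, swapping the priority of two post-process components so the selected view wins.
; It writes Enable to the OutofRangeMode parameter collection, and range, threshold, opacity and color settings to the FalseColorMode collection.
; BeginPlay, overlap and Tick events are empty, so all logic runs at construction.
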(event EventBeginPlay)

(event Collision|EventActorBeginOverlap
  (params (OtherActor :type Actor%20Object%20Reference)))

(event EventTick
  (params (DeltaSeconds :type Float%20%28single-precision%29)))

(fn ConstructionScript
  (bind _postprocessvolumn_falsecolor (Variables|Default|GetPostProcessVolumn-FalseColor))
  (bind _postprocessvolumn___outofrange (Variables|Default|GetPostProcessVolumn%20-%20OutofRange))
  (Rendering|Material|SetScalarParameterValue "/Game/ArtTools/HDR_Checker/MPC_OutofRangeMode.MPC_OutofRangeMode" "Enable" (Math|Conversions|ToFloat(Boolean) (Variables|Default|GetEnable)) :!class /Script/Engine.KismetMaterialLibrary)
  (switch Utilities|FlowControl|Switch|SwitchonEnum_ViewModes (Variables|Default|GetViewMode)
    (:NewEnumerator0
      (Variables|/Script/Engine.PostProcessComponent|SetPriority 0.0 _postprocessvolumn___outofrange)
      (Variables|/Script/Engine.PostProcessComponent|SetPriority 1e+02 _postprocessvolumn_falsecolor))
    (:NewEnumerator1
      (Variables|/Script/Engine.PostProcessComponent|SetPriority 1e+02 _postprocessvolumn___outofrange)
      (Variables|/Script/Engine.PostProcessComponent|SetPriority 0.0 _postprocessvolumn_falsecolor)))
  (Rendering|Material|SetScalarParameterValue "/Game/ArtTools/HDR_Checker/MPC_FalseColorMode.MPC_FalseColorMode" "FalseColorRange" (Variables|Default|GetFalseColor%20-%20Range) :!class /Script/Engine.KismetMaterialLibrary)
  (Rendering|Material|SetScalarParameterValue "/Game/ArtTools/HDR_Checker/MPC_FalseColorMode.MPC_FalseColorMode" "Threshold Value" (Variables|Default|GetOutofRange%20-%20Threshold%20Value) :!class /Script/Engine.KismetMaterialLibrary)
  (Rendering|Material|SetScalarParameterValue "/Game/ArtTools/HDR_Checker/MPC_FalseColorMode.MPC_FalseColorMode" "ThreshColorOpacity" (Variables|Default|GetOutofRange%20-%20Opacity) :!class /Script/Engine.KismetMaterialLibrary)
  (Rendering|Material|SetVectorParameterValue "/Game/ArtTools/HDR_Checker/MPC_FalseColorMode.MPC_FalseColorMode" "Debug_Color" (Variables|Default|GetOutofRange%20-%20Threshold%20Color) :!class /Script/Engine.KismetMaterialLibrary))
