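# Material function in the HDR_Checker art tools that turns the horizontal UV coordinate into a blue-cyan-green-yellow-red rainbow gradient.
# Each channel is a piecewise linear ramp built from If, Multiply and Add nodes with breakpoints at quarter intervals.
# Red rises late, green is the product of rising and falling ramps, and blue falls early, then all three are appended as RGB.
# A HueShift branch on pure red ends in a reroute and does not feed the output.
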MaterialFunction /Game/ArtTools/HDR_Checker/MF_Rainbow
graph {nodes [
  n1 = FunctionInput() {inputName=UV, inputType=FunctionInput_Vector2}
  n2 = ComponentMask(None=@n1) {r=true}
  n3 = Add(A=Multiply(A=@n2) {constB=4}) {constB=-2}
  n4 = Clamp(None=If(A=@n2, B=Constant() {r=0.5}, "A > B"=@n3, "A == B"=@n3, "A < B"=Constant()))
  n5 = ComponentMask(None=@n1) {r=true}
  n6 = Constant() {r=1}
  n7 = Multiply(A=@n5) {constB=-4}
  n8 = Constant() {r=1}
  n9 = Multiply(
   A=If(A=@n5, B=Constant() {r=0.25}, "A > B"=@n6, "A == B"=@n6, "A < B"=Multiply(A=@n5) {constB=4})
   B=If(A=@n5, B=Constant() {r=0.75}, "A > B"=Add(A=@n7) {constB=4}, "A == B"=@n8, "A < B"=@n8)
  )
  n10 = ComponentMask(None=@n1) {r=true}
  n11 = Add(A=Multiply(A=@n10) {constB=-4}) {constB=2}
  n12 = If(A=@n10, B=Constant() {r=0.25}, "A > B"=@n11, "A == B"=@n11, "A < B"=Constant() {r=1})
  FunctionOutput(None=AppendVector(A=AppendVector(A=@n4, B=@n9), B=Clamp(None=@n12)))
  Reroute(None=MaterialFunctionCall(Hue Shift Percentage=Multiply(A=@n1) {constB=0.8}, Texture=Constant3Vector() {constant=<rgba 1 0 0 0>}) {materialFunction=&HueShift}.Result)
 ]}
