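# Post-process debug material that checks HDR scene color after depth of field, either as false color or as an out-of-range highlight.
# With FalseMode? on, desaturated scene color divided by FalseColorRange drives a rainbow ramp, with UV-cropping functions overlaying a bordered legend bar low on screen.
# With it off, pixels above a threshold blend toward a debug color.
# Material Parameter Collections supply the settings and an Enable blend over the original image.
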Material /Game/ArtTools/HDR_Checker/HDR_PostPocessingMaterial
graph {
 properties {
  blendableLocation=BL_SceneColorAfterDOF
  blendablePriority=1
  materialDomain=MD_PostProcess
  pixelDepthOffsetMode=PDOM_Legacy
  usedShadingModels=Default Lit
 }
 nodes [
  n1 = Constant() {r=0.0075}
  n2 = Constant() {r=0.475}
  n3 = MaterialFunctionCall(Float2=TextureCoordinate()) {materialFunction=&BreakOutFloat2Components}
  n4 = Add(A=@n3.G) {constB=-0.4, desc=Reposition the Bar}
  n5 = AppendVector(A=@n3.R, B=@n4)
  n6 = SceneTexture() {sceneTextureId=PPI_PostProcessInput0}
  n7 = Divide(
   A=MaterialFunctionCall(Color=Desaturation(None=@n6.Color)) {materialFunction=&SplitComponents}.RGB
   B=CollectionParameter() {collection=&/MPC_FalseColorMode, parameterName=FalseColorRange}
  ) {constB=10}
  n8 = Multiply(
   A=OneMinus(None=MaterialFunctionCall!(Crop Right=@n1, Crop Bottom=@n2, Crop Top=@n2, Input Texcord=@n5, Crop Left=@n1) {materialFunction=&/UVCropping_withUVInput}.Result)
   B=MaterialFunctionCall!(UV=Clamp(None=@n7)) {materialFunction=&/MF_Rainbow}.Result
  )
  n9 = MaterialFunctionCall(UV=Clamp(None=@n3.R)) {materialFunction=&MF_Rainbow}
  n10 = Constant() {r=0.01}
  n11 = Constant() {r=0.48}
  n12 = MaterialFunctionCall(Crop Right=@n10, Crop Bottom=@n11, Crop Top=@n11, Input Texcord=AppendVector(A=@n3.R, B=@n4), Crop Left=@n10) {materialFunction=&UVCropping_withUVInput}
  n13 = MaterialFunctionCall(Color=@n6.Color) {materialFunction=&SplitComponents}
  n14 = Desaturation(None=@n13.RGB)
  n15 = CollectionParameter() {collection=&MPC_FalseColorMode, parameterName=Threshold Value}
  n16 = LinearInterpolate(
   A=@n13.RGB
   B=MaterialFunctionCall(Color=CollectionParameter() {collection=&MPC_FalseColorMode, parameterName=Debug_Color}) {materialFunction=&SplitComponents}.RGB
   Alpha=CollectionParameter() {collection=&MPC_FalseColorMode, parameterName=ThreshColorOpacity}
  )
  LinearInterpolate(
   A=@n13.RGB
   B=StaticSwitchParameter(True=LinearInterpolate(A=@n8, B=@n9.Result, Alpha=@n12.Result), False=If(A=@n14, B=@n15, "A > B"=@n16, "A == B"=@n13.RGB, "A < B"=@n13.RGB)) {parameterName=FalseMode?}
   Alpha=CollectionParameter() {collection=&/MPC_OutofRangeMode, parameterName=Enable}
  )
 ]
}
